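; Supply drop radio actor component that temporarily overrides and restores the net cull distance of the actor it is attached to.
; On BeginPlay the server caches the owner's original squared cull distance.
; Authority-only Set and Reset functions write a new or cached value into a RepNotify variable and log it.
; The OnRep pushes positive values to the owner through SetOwningActorNetCullDistance, after a cast to the Fortnite replication graph interface.
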(event EventBeginPlay
  (bind _returnvalue (Components|GetOwner))
  ;[0 0.07292 0.005071] BeginPlay: Cache the original net cull distance squared value.
  (|CallParentFunction)
  (Networking|SwitchHasAuthority
    (:Authority
      (Utilities|IsValid _returnvalue
        (:Is%20Valid
          (bind _output_get (Variables|Default|SetOriginalNetCullDistanceSquared (Variables|/Script/Engine.Actor|GetNetCullDistanceSquared _returnvalue)))
          (Fortnite|FortShippingLog self (Utilities|String|Append "BeginPlay: Caching original net cull distance sq as " (Utilities|String|ToString(Float) _output_get) " for actor " (Utilities|GetDisplayName _returnvalue))))))))

(fn OnRep_NetCullDistanceSquared
  ;[0.03561 0.1657 0.1927] OnRep_NetCullDistanceSquared: Update the network cull distance to the desired value.
  (Utilities|Casting|CastToFortReplicationGraphInterface (Components|GetOwner))
  (if (> (Variables|Default|GetNetCullDistanceSquared) 0)
    (SupplyDropRadio|SetOwningActorNetCullDistance (Variables|Default|GetNetCullDistanceSquared))))

(fn ResetNetCullDistanceSquared
  ;[0.01819 0.1429 0.1458] Reset the net cull distance of the owner to the default value.
  (Networking|SwitchHasAuthority
    (:Authority
      (bind _output_get (Variables|Default|SetNetCullDistanceSquared (Variables|Default|GetOriginalNetCullDistanceSquared)))
      (Fortnite|FortShippingLog self (Utilities|String|Append "ResetNetCullDistanceSquared: Resetting net cull distance sq as " (Utilities|String|ToString(Float) _output_get) " for actor " (Utilities|GetDisplayName (Components|GetOwner)))))))

(fn SetNetCullDistanceSquared
  (params (InNetCullDistanceSquared :type Float%20%28double-precision%29))
  ;[0.01819 0.1429 0.1458] Update the net cull distance squared value.
  ; The OnRep function will push it on the owning actor.
  (Networking|SwitchHasAuthority
    (:Authority
      (bind _output_get (Variables|Default|SetNetCullDistanceSquared InNetCullDistanceSquared))
      (Fortnite|FortShippingLog self (Utilities|String|Append "SetNetCullDistanceSquared: Setting net cull distance sq as " (Utilities|String|ToString(Float) _output_get) " for actor " (Utilities|GetDisplayName (Components|GetOwner)))))))
