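; Game state component that manages Supply Drop Radio options, shared stock and end-of-match analytics.
; On authority, it picks a datatable by playlist tags and builds replicated stock entries, with negative stock meaning unlimited.
; Pawns reserve an option by conversation index, decrementing stock; reservations are refunded or handed to a flare actor, updating usage counts and quest stats.
; Array element writes force replication, and the RepNotify broadcasts updates.
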(event EventBeginPlay
  ;[0.01734 0.2135 0.02705] BeginPlay: Process the supply drop radio data specified in the datatable.
  (|CallParentFunction)
  (Networking|SwitchHasAuthority
    (:Authority
      (bind (_gamestate _playlist _playlistcontexttags) (Class|FortAsyncActionCurrentPlaylistReady|PlaylistDataReadyAsync)
        (:OnReady
          (Variables|Transient|SetPlaylistTags _playlistcontexttags)
          (CallFunction|CacheSupplyDropRadioDatatable)
          (CallFunction|ProcessSupplyDropRadioData)
          (bind _returnvalue (GameplayEventRouter|GetEventRouter "/Script/FortniteGame.FortGameState" (Components|GetOwner)))
          (bind (_asynctaskproxy _payload) (GameplayEventRouter|StartListeningToStatefulEvent _returnvalue "/Script/FortniteGame.FortGamePhaseUpdatedEvent")
            (:then
              (Variables|Default|SetGamePhaseUpdatedAsyncTask _asynctaskproxy))
            (:OnEventReceived
              _)
            (:OnSavedEventStateExists
              _))
          (CallFunction|HandleGameStatePhaseChanged (Utilities|Struct|BreakFortGamePhaseUpdatedEvent _payload)))) ; Wait for hotfix datatables to be ready.
    )))

(event Custom|HandleGameStatePhaseChanged
  (params (GamePhase :type EAthenaGamePhase%20Enum))
  ;[0.02875 0.1979 0.192] HandleGameStatePhasedChange: Report analytics at the end of the match.
  (Networking|SwitchHasAuthority
    (:Authority
      (if (Utilities|Enum|Equal(Enum) GamePhase "EndGame")
        (Analytics|ReportEndGameAnalytics)
        (bind _gamephaseupdatedasynctask (Variables|Default|GetGamePhaseUpdatedAsyncTask)
          (:then
            (GameplayEventRouter|StopListeningtoStatefulEvent _gamephaseupdatedasynctask))
          (:else))))))

(event EventEndPlay
  (params (EndPlayReason :type EEndPlayReason%20Enum :default "Destroyed"))
  ;[0.2188 0.1331 0.04406] EndPlay: Unbind delegates.
  (|CallParentFunction EndPlayReason)
  (Networking|SwitchHasAuthority
    (:Authority
      (bind _gamephaseupdatedasynctask (Variables|Default|GetGamePhaseUpdatedAsyncTask)
        (:then
          (GameplayEventRouter|StopListeningtoStatefulEvent _gamephaseupdatedasynctask))
        (:else)))))

(event Custom|ResetAllStock
  ;[0 0.1823 0.1479] ResetAllStocks: Refreshes the stock count from the datatable.
  (Networking|SwitchHasAuthority
    (:Authority
      (CallFunction|ProcessSupplyDropRadioData))))

(event Custom|ResetStock
  (params (GameplayTag :type Gameplay%20Tag%20Structure :default "(TagName=\"SupplyDropRadio.SupplyDrop.Type.Health\")"))
  ;[0.01639 0.136 0.151] ResetStock: Resets the stock of a particular item.
  (Networking|SwitchHasAuthority
    (:Authority
      (bind _supplydropradiodatatable (Variables|Default|GetSupplyDropRadioDatatable)
        (:then
          (bind _outrownames (DataTable|GetDataTableRowNames _supplydropradiodatatable))
          (bind _array_element (Utilities|Array|ForEachLoopwithBreak _outrownames)
            (:LoopBody
              (bind _returnvalue (Utilities|GetDataTableRow _supplydropradiodatatable _array_element))
              (if (== (Utilities|Struct|BreakSupplyDropRadioDataStruct _returnvalue) GameplayTag)
                (bind (_array_element_2 _array_index) (Utilities|Array|ForEachLoopwithBreak (Variables|Transient|GetSupplyDropRadioRepData))
                  (:LoopBody
                    (if (== (Utilities|Struct|BreakSupplyDropRadioRepDataStruct _array_element_2) GameplayTag)
                      (Utilities|Array|SetArrayElem (Variables|Transient|GetSupplyDropRadioRepData) _array_index (Utilities|Struct|MakeSupplyDropRadioRepDataStruct GameplayTag (.Stock_34_7D54BF96461297E10D7DA1AA8FD4C012 (Utilities|Struct|BreakSupplyDropRadioDataStruct _returnvalue))))))
                  (:Completed
                    _))))))
        (:else)))))

(fn Analytics_IncrementPlayerDestroyedFallingBalloon
  ;[0.03125 0.1897 0.2344] Increment the number of supply drop balloons that were destroyed by a player.
  (Networking|SwitchHasAuthority
    (:Authority
      (Math|Integer|IncrementInt (Variables|Transient|Analytics|GetAnalytics_PlayerDestroyedFallingBalloons)))))

(fn CacheSupplyDropRadioDatatable
  ;[0.1127 0.0589 0.1771] CacheSupplyDropRadioDatatable: Cache the datatable to use based on the playlist that is active.
  (Networking|SwitchHasAuthority
    (:Authority
      (bind _array_element (Utilities|Array|ForEachLoopwithBreak (Variables|Default|GetPlaylistDatatables))
        (:LoopBody
          (if (not (GameplayTags|IsTagQueryEmpty (Utilities|Struct|BreakSupplyDropRadioPlaylistDatatableStruct _array_element)))
            (if (GameplayTags|DoesContainerMatchTagQuery (Variables|Transient|GetPlaylistTags) (Utilities|Struct|BreakSupplyDropRadioPlaylistDatatableStruct _array_element))
              (Utilities|IsValid (.Datatable_5_C7E1168E435BAFB05233FDA129300904 (Utilities|Struct|BreakSupplyDropRadioPlaylistDatatableStruct _array_element))
                (:Is%20Valid
                  (Variables|Default|SetSupplyDropRadioDatatable (.Datatable_5_C7E1168E435BAFB05233FDA129300904 (Utilities|Struct|BreakSupplyDropRadioPlaylistDatatableStruct _array_element))))))))
        (:Completed
          _))
      (Fortnite|FortShippingLog self (Utilities|String|Append "CacheSupplyDropRadioDatatable: " (Utilities|GetDisplayName (Variables|Default|GetSupplyDropRadioDatatable))) (not (Utilities|IsValid (Variables|Default|GetSupplyDropRadioDatatable) :!class /Script/Engine.KismetSystemLibrary))))))

(fn GetReservedSupplyDropRadioDataForPlayer
  (params (PlayerPawn :type Fort%20Player%20Pawn%20Object%20Reference))
  (bind _authoritysupplydropreservationmap (Variables|Transient|GetAuthoritySupplyDropReservationMap))
  (bind (_value _returnvalue) (Utilities|Map|Find _authoritysupplydropreservationmap PlayerPawn))
  ;[0.04064 0.1531 0.1562] Return the supply drop radio data that is reserved for the player.
  (Utilities|FlowControl|Sequence
    (:then_0
      (Utilities|IsValid PlayerPawn
        (:Is%20Valid
          (if _returnvalue
            (bind (_isvalid _data) (CallFunction|GetSupplyDropRadioDataUsingConversationIndex _value))
            (return _isvalid _data)))))
    (:then_1
      (return false 0))))

(fn GetSupplyDropRadioDataUsingConversationIndex
  (params (ConversationIndex :type Integer))
  (bind _supplydropradiorepdata (Variables|Transient|GetSupplyDropRadioRepData))
  ;[0.02894 0.1085 0.1615] Returns the supply drop radio data for a particular conversation index.
  (Utilities|FlowControl|Sequence
    (:then_0
      (if (Utilities|Array|IsValidIndex _supplydropradiorepdata ConversationIndex)
        (bind (_isvalid _data) (CallFunction|GetSupplyDropRadioDataUsingGameplayTag (Utilities|Struct|BreakSupplyDropRadioRepDataStruct (Utilities|Array|Get(aref) _supplydropradiorepdata ConversationIndex))))
        (return _isvalid _data)))
    (:then_1
      (return false 0))))

(fn GetSupplyDropRadioDataUsingGameplayTag
  (params (GameplayTag :type Gameplay%20Tag%20Structure))
  ;[0.02272 0.1135 0.1458] Returns the supply drop radio data for a particular gameplay tag.
  (Utilities|FlowControl|Sequence
    (:then_0
      (bind _supplydropradiodatatable (Variables|Default|GetSupplyDropRadioDatatable)
        (:then
          (bind _outrownames (DataTable|GetDataTableRowNames _supplydropradiodatatable))
          (for _array_element _outrownames
            (bind _returnvalue (Utilities|GetDataTableRow _supplydropradiodatatable _array_element))
            (if (== (.GameplayTag_5_542794F64B7B42F5AFFB4F99DE156B3F (Utilities|Struct|BreakSupplyDropRadioDataStruct _returnvalue)) GameplayTag)
              (return true _returnvalue))))
        (:else)))
    (:then_1
      (return false 0))))

(fn IsConversationIndexEnabled
  (params (ConversationIndex :type Integer))
  ;[0.02755 0.06363 0.1719] Returns if a particular conversation index has valid data to show.
  (return (Utilities|Array|IsValidIndex (Variables|Transient|GetSupplyDropRadioRepData) ConversationIndex)))

(fn IsStockAvailableForConversationIndex
  (params (ConversationIndex :type Integer))
  (bind (_gameplaytag_5_542794f64b7b42f5affb4f99de156b3f _stock_21_7d54bf96461297e10d7da1aa8fd4c012) (Utilities|Struct|BreakSupplyDropRadioRepDataStruct (Utilities|Array|Get(aref) (Variables|Transient|GetSupplyDropRadioRepData) ConversationIndex)))
  (bind _supplydropradiorepdata (Variables|Transient|GetSupplyDropRadioRepData))
  ;[0.1797 0.07268 0.2135] Returns if there's stock available for the conversation index.
  (Utilities|FlowControl|Sequence
    (:then_0
      (if (Utilities|Array|IsValidIndex _supplydropradiorepdata ConversationIndex)
        (return (!= _stock_21_7d54bf96461297e10d7da1aa8fd4c012 0) _stock_21_7d54bf96461297e10d7da1aa8fd4c012)))
    (:then_1
      (return false 0))))

(fn OnRep_SupplyDropRadioRepData
  ;[0.1762 0.04975 0.1979] Broadcast that the supply drop radio rep data has been updated.
  (Default|CallOnSupplyDropRadioRepDataUpdated))

(fn ProcessSupplyDropRadioData
  ;[0.03683 0.276 0.1738] ProcessSupplyDropRadioData: Run through the supply drop radio datatable and create the replicated information for all the options that are enabled and valid.
  (Networking|SwitchHasAuthority
    (:Authority
      (Utilities|Array|Clear (Variables|Transient|GetSupplyDropRadioRepData))
      (bind _supplydropradiodatatable (Variables|Default|GetSupplyDropRadioDatatable)
        (:then
          (bind _outrownames (DataTable|GetDataTableRowNames _supplydropradiodatatable))
          (for _array_element _outrownames
            (bind _returnvalue (Utilities|GetDataTableRow _supplydropradiodatatable _array_element))
            (if (Math|Boolean|ANDBoolean (Utilities|Struct|BreakSupplyDropRadioDataStruct _returnvalue) (or (GameplayTags|DoesContainerMatchTagQuery (Variables|Transient|GetPlaylistTags) (.PlaylistTagsQuery_39_183349BE420E6A9770D562ABF5C5D3EB (Utilities|Struct|BreakSupplyDropRadioDataStruct _returnvalue))) (GameplayTags|IsTagQueryEmpty (.PlaylistTagsQuery_39_183349BE420E6A9770D562ABF5C5D3EB (Utilities|Struct|BreakSupplyDropRadioDataStruct _returnvalue)))) (GameplayTags|IsGameplayTagValid (.GameplayTag_5_542794F64B7B42F5AFFB4F99DE156B3F (Utilities|Struct|BreakSupplyDropRadioDataStruct _returnvalue))) (Utilities|IsValidSoftClassReference (.SupplyDropRadioActorClass_58_71A3A84043A4401BC875CFA344AAF3DD (Utilities|Struct|BreakSupplyDropRadioDataStruct _returnvalue))))
              (Utilities|Array|Add (Variables|Transient|GetSupplyDropRadioRepData) (Utilities|Struct|MakeSupplyDropRadioRepDataStruct (.GameplayTag_5_542794F64B7B42F5AFFB4F99DE156B3F (Utilities|Struct|BreakSupplyDropRadioDataStruct _returnvalue)) (.Stock_34_7D54BF96461297E10D7DA1AA8FD4C012 (Utilities|Struct|BreakSupplyDropRadioDataStruct _returnvalue))))
              (Utilities|Map|Add (Variables|Transient|Analytics|GetAnalytics_SupplyDropsRequestedMap) (.GameplayTag_5_542794F64B7B42F5AFFB4F99DE156B3F (Utilities|Struct|BreakSupplyDropRadioDataStruct _returnvalue))))))
        (:else)))))

(fn RefundSupplyDrop
  (params (RequestingPlayerPawn :type Fort%20Player%20Pawn%20Object%20Reference))
  (bind _supplydropradiorepdata (Variables|Transient|GetSupplyDropRadioRepData))
  (bind _output (Utilities|Array|Get(aref) _supplydropradiorepdata (Utilities|Map|Find (Variables|Transient|GetAuthoritySupplyDropReservationMap) RequestingPlayerPawn)))
  (bind (_gameplaytag_5_542794f64b7b42f5affb4f99de156b3f _stock_21_7d54bf96461297e10d7da1aa8fd4c012) (Utilities|Struct|BreakSupplyDropRadioRepDataStruct _output))
  (bind _authoritysupplydropreservationmap (Variables|Transient|GetAuthoritySupplyDropReservationMap))
  (bind (_value _returnvalue) (Utilities|Map|Find _authoritysupplydropreservationmap RequestingPlayerPawn))
  ;[0.02414 0.02894 0.09375] RefundSupplyDrop: This function reverts a reservation of a supply drop made by a player. For e.g., if the player's confirm animation fails to play to the point where the supply drop is supposed to spawn.
  (Networking|SwitchHasAuthority
    (:Authority
      (Utilities|IsValid RequestingPlayerPawn
        (:Is%20Valid
          (if _returnvalue
            (Utilities|Map|Remove _authoritysupplydropreservationmap (Variables|GetRequestingPlayerPawn))
            (if (Utilities|Array|IsValidIndex _supplydropradiorepdata _value)
              (if (>= _stock_21_7d54bf96461297e10d7da1aa8fd4c012 0)
                (bind _structout (Utilities|Struct|SetmembersinSupplyDropRadioRepDataStruct _output (+ _stock_21_7d54bf96461297e10d7da1aa8fd4c012 1)))
                (Utilities|Array|SetArrayElem (Variables|Transient|GetSupplyDropRadioRepData) _value _structout) ; Setting it forces replication.
              ))))))))

(fn ReportEndGameAnalytics
  (bind _analytics_supplydropsrequestedmap (Variables|Transient|Analytics|GetAnalytics_SupplyDropsRequestedMap))
  ;[0.01632 0.06666 0.1042] ReportEndGameAnalytics: Report the analytics of the supply drop radio at the end of the match.
  (Networking|SwitchHasAuthority
    (:Authority
      (if (not (Replay|GetIsPlayingReplay))
        (Utilities|FlowControl|Sequence
          (:then_0
            (Utilities|FlowControl|Sequence
              (:then_0
                ;;[1 1 1] SupplyDropRadioUsed
                (bind _returnvalue (Analytics|ConstructAnalyticsAttribute :AttributeName "SupplyDropRadioUsed" :IntValue (Variables|Transient|Analytics|GetAnalytics_SupplyDropRadioUsed) :floatValue -1.0 :BoolValue false))
                (Utilities|Array|Add (LocalVariables|GetAnalyticsEventAttributes) _returnvalue)
                ;;[1 1 1] SupplyDropRadioPlayerDestroyedFallingBalloons
                (bind _returnvalue_2 (Analytics|ConstructAnalyticsAttribute :AttributeName "SupplyDropRadioPlayerDestroyedFallingBalloons" :IntValue (Variables|Transient|Analytics|GetAnalytics_PlayerDestroyedFallingBalloons) :floatValue -1.0 :BoolValue false))
                (Utilities|Array|Add (LocalVariables|GetAnalyticsEventAttributes) _returnvalue_2))
              (:then_1
                ;;[1 1 1] SupplyDropRadioRequested
                (bind _keys (Utilities|Map|Keys _analytics_supplydropsrequestedmap))
                (for _array_element _keys
                  (if (.ReturnValue (Utilities|Map|Find _analytics_supplydropsrequestedmap _array_element))
                    (bind _returnvalue_3 (Analytics|ConstructAnalyticsAttribute :AttributeName "SupplyDropRadioRequested" :StringValue (GameplayTags|GetDebugStringfromGameplayTag _array_element) :IntValue (Utilities|Map|Find _analytics_supplydropsrequestedmap _array_element) :floatValue -1.0 :BoolValue false))
                    (Utilities|Array|Add (LocalVariables|GetAnalyticsEventAttributes) _returnvalue_3))))))
          (:then_1
            (Analytics|FireAnalyticsEventBlueprintArray "Core.SupplyDropRadio" (LocalVariables|GetAnalyticsEventAttributes))))))))

(fn ReserveSupplyDrop
  (params (RequestingPlayerPawn :type Fort%20Player%20Pawn%20Object%20Reference) (ConversationIndex :type Integer))
  (bind (_gameplaytag_5_542794f64b7b42f5affb4f99de156b3f _stock_21_7d54bf96461297e10d7da1aa8fd4c012) (Utilities|Struct|BreakSupplyDropRadioRepDataStruct (Utilities|Array|Get(aref) (Variables|Transient|GetSupplyDropRadioRepData) ConversationIndex)))
  (bind (_value _returnvalue) (Utilities|Map|Find (Variables|Transient|GetAuthoritySupplyDropReservationMap) RequestingPlayerPawn))
  ;[0.1198 0.04399 0.1161] ReserveSupplyDrop: This logic reserves a supply drop based on the request made by a player pawn.
  (Utilities|FlowControl|Sequence
    (:then_0
      (Networking|SwitchHasAuthority
        (:Authority
          (Utilities|IsValid RequestingPlayerPawn
            (:Is%20Valid
              (if (not _returnvalue)
                (if (Utilities|Array|IsValidIndex (Variables|Transient|GetSupplyDropRadioRepData) ConversationIndex)
                  (if (!= _stock_21_7d54bf96461297e10d7da1aa8fd4c012 0)
                    (Utilities|Map|Add (Variables|Transient|GetAuthoritySupplyDropReservationMap) RequestingPlayerPawn ConversationIndex)
                    (bind _structout (Utilities|Struct|SetmembersinSupplyDropRadioRepDataStruct (Utilities|Array|Get(aref) (Variables|Transient|GetSupplyDropRadioRepData) ConversationIndex) (Math|Integer|Max(Integer) (- (Utilities|Struct|BreakSupplyDropRadioRepDataStruct (Utilities|Array|Get(aref) (Variables|Transient|GetSupplyDropRadioRepData) ConversationIndex)) 1) -1)))
                    (Utilities|Array|SetArrayElem (Variables|Transient|GetSupplyDropRadioRepData) ConversationIndex _structout) ; Setting it forces replication.
                    (return true)))))))))
    (:then_1
      (return false))))

(fn SetReservedSupplyDropToFlare
  (params (RequestingPlayerPawn :type Fort%20Player%20Pawn%20Object%20Reference) (RequestingFlareActor :type BP%20Supply%20Drop%20Radio%20Flare%20Actor%20Object%20Reference))
  (bind _authoritysupplydropreservationmap (Variables|Transient|GetAuthoritySupplyDropReservationMap))
  (bind (_value _returnvalue) (Utilities|Map|Find _authoritysupplydropreservationmap RequestingPlayerPawn))
  (bind _requestingplayerpawn (Variables|GetRequestingPlayerPawn))
  (bind _returnvalue_2 (Fortnite|GetFortPlayerControllerfromActor _requestingplayerpawn))
  (bind _returnvalue_3 (Quests|GetQuestManager _returnvalue_2 "Invalid"))
  (bind _requestingflareactor (Variables|GetRequestingFlareActor))
  ;[0.125 0.02638 0.1047] SpawnReservedSupplyDrop: This function actually ends up spawning the supply drop actor in the world based on a reservation made by a player.
  (Utilities|FlowControl|Sequence
    (:then_0
      (Networking|SwitchHasAuthority
        (:Authority
          (Utilities|IsValid RequestingPlayerPawn
            (:Is%20Valid
              (if _returnvalue
                (bind (_isvalid _data) (CallFunction|GetSupplyDropRadioDataUsingConversationIndex _value))
                (if _isvalid
                  (if (Utilities|IsValidSoftClassReference (.SupplyDropRadioActorClass_58_71A3A84043A4401BC875CFA344AAF3DD (Utilities|Struct|BreakSupplyDropRadioDataStruct _data)))
                    (Utilities|IsValid _requestingflareactor
                      (:Is%20Valid
                        (Class|BPSupplyDropRadioFlareActor|SetSupplyDropRadioConversationIndex _requestingflareactor _value))
                      (:Is%20Not%20Valid
                        _))
                    (Utilities|Map|Remove (Variables|Transient|GetAuthoritySupplyDropReservationMap) (Variables|GetRequestingPlayerPawn))
                    (Math|Integer|IncrementInt (Variables|Transient|Analytics|GetAnalytics_SupplyDropRadioUsed)
                      (:%20%20
                        (if (Utilities|Array|IsValidIndex (Variables|Transient|GetSupplyDropRadioRepData) _value)
                          (Utilities|Map|Add (Variables|Transient|Analytics|GetAnalytics_SupplyDropsRequestedMap) (Utilities|Struct|BreakSupplyDropRadioRepDataStruct (Utilities|Array|Get(aref) (Variables|Transient|GetSupplyDropRadioRepData) _value)) (+ (Utilities|Map|Find (Variables|Transient|Analytics|GetAnalytics_SupplyDropsRequestedMap) (Utilities|Struct|BreakSupplyDropRadioRepDataStruct (Utilities|Array|Get(aref) (Variables|Transient|GetSupplyDropRadioRepData) _value))) 1))
                          (else
                            _))
                        (if (Utilities|GameplayTagContainerNotEmpty (.QuestTags_44_0A989AC04A6AFB5EFD32CCA67362B3E7 (Utilities|Struct|BreakSupplyDropRadioDataStruct _data)))
                          (Utilities|IsValid _returnvalue_2
                            (:Is%20Valid
                              (Utilities|IsValid _returnvalue_3
                                (:Is%20Valid
                                  (Quest|SendComplexCustomStatEvent _returnvalue_3 0 (.QuestTags_44_0A989AC04A6AFB5EFD32CCA67362B3E7 (Utilities|Struct|BreakSupplyDropRadioDataStruct _data)) "(GameplayTags=)" 1))
                                (:Is%20Not%20Valid
                                  _))
                              (return true))
                            (:Is%20Not%20Valid
                              _))
                          (else
                            _))))))))))))
    (:then_1
      (return false))))
